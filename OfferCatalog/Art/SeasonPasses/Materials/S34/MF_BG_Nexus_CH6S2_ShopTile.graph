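# Material function that builds the animated Chapter 6 Season 2 Nexus background for a season pass shop tile.
# Focal-point UVs drive gradients, pulsing corner lights, city light textures, alternating red/blue police lights and cloud-noise grunge.
# A glass texture sampled in polar coordinates with explicit derivatives supplies cracks, with bump-offset highlights and a moving shine.
# Named reroutes organize these elements into Background, Middleground and Foreground stages, composited mainly with color dodge and lerps.
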MaterialFunction /OfferCatalog/Art/SeasonPasses/Materials/S34/MF_BG_Nexus_CH6S2_ShopTile
graph {nodes [
  FunctionOutput(None=NamedRerouteUsage())
  n1 = MaterialFunctionCall() {materialFunction=&MF_UI_UVScaling}
  n2 = AppendVector(A=ScalarParameter() {group=Focal Point, parameterName=FocalX}, B=ScalarParameter() {group=Focal Point, parameterName=FocalY})
  NamedRerouteDeclaration(None=MaterialFunctionCall(UVs=Add(A=@n1.Scaled To Fill, B=@n2), Texture Scale=ScalarParameter() {defaultValue=1, group=Focal Point, parameterName=FocalSize}) {materialFunction=&ScaleUVsByCenter}.UVs) {name=UV, nodeColor=<rgba 0 0 0 1>}
  n3 = NamedRerouteUsage()
  n4 = NamedRerouteUsage()
  n5 = ScalarParameter() {defaultValue=0.6, group=Base Gradient, parameterName=Crack Faces Opacity}
  n6 = MaterialFunctionCall(
   UVs=Add(A=NamedRerouteUsage(), B=Multiply(A=@n4, B=@n5) {constB=0.6})
   RotationAngle=ScalarParameter() {defaultValue=0.15, group=Base Gradient, parameterName=Rotation}
  ) {materialFunction=&MF_UI_LinearGradient}
  n7 = LinearInterpolate(
   A=LinearInterpolate(B=Constant3Vector() {constant=<rgba 0.0363 0.0461 0.0625 1>}, Alpha=Power(Base=@n6.Result) {constExponent=2.5})
   B=LinearInterpolate(A=Constant3Vector() {constant=<rgba 0.0993 0.075 0.1823 1>}, B=Constant3Vector() {constant=<rgba 0.0117 0.0165 0.0365 1>}, Alpha=@n3)
   Alpha=@n3
  )
  n8 = Power(Base=Multiply(A=NamedRerouteUsage(), B=Constant3Vector() {constant=<rgba 0.101 0.0947 0.3385 1>})) {constExponent=0.85}
  NamedRerouteDeclaration(None=Saturate(None=Add(A=@n7, B=@n8))) {name=Base Gradient, nodeColor=<rgba 0 0 0 1>}
  n9 = ScreenPosition()
  n10 = Constant() {r=0.2}
  n11 = Multiply(
   A=NamedRerouteUsage()
   B=LinearInterpolate(Alpha=MaterialFunctionCall(Speed=@n10) {materialFunction=&DebugTimeSine}.Result) {constA=0.1, constB=0.2}
  ) {constB=0.5}
  n12 = MaterialFunctionCall(UVs=Add(A=Add(A=@n9.ViewportUV, B=@n11), B=Constant2Vector() {g=0.5, r=-0.5}), Texture Scale=Constant2Vector() {g=4, r=4}) {materialFunction=&ScaleUVsByCenter}
  n13 = Time()
  n14 = Constant() {r=0.1}
  n15 = Power(
   Base=Power(Base=MaterialFunctionCall(UV=@n12.UVs) {materialFunction=&MF_UI_RadialGradient}.Result) {constExponent=3}
   Exp=LinearInterpolate(Alpha=MaterialFunctionCall!(Value=@n13, Scale=@n14, Height) {materialFunction=/FortUILibrary/MaterialFunctions/Basics/MF_UI_PingPong}.Result) {constA=1, constB=2}
  )
  n16 = Multiply(
   A=MaterialFunctionCall!(UV=ScreenPosition().ViewportUV, Seed) {materialFunction=&MF_UI_Random}.Result
   B=@n15
  )
  NamedRerouteDeclaration(None=Add(A=@n15, B=Multiply(A=@n16) {constB=0.05})) {name=Right Light, nodeColor=<rgba 0 0 0 1>}
  NamedRerouteDeclaration(None=MaterialFunctionCall(
    UVs=MaterialFunctionCall!(Vault World Color, Focal Color Override, Focal Scale Override, Mask Blend) {materialFunction=/Game/Environments/FrontEnd/Blueprints/ItemPreview/VaultWorldTransitions/MaterialFunctions/MF_UI_VaultWorld_FocalPoint}.Focal Mask
    Texture Scale=Constant() {r=0.8}
   ) {materialFunction=&ScaleUVsByCenter}.UVs) {name=Focal Mask, nodeColor=<rgba 0 0 0 1>}
  n17 = NamedRerouteUsage()
  n18 = AppendVector(
   A=ScalarParameter() {defaultValue=-0.1, parameterName=Glass Position X}
   B=ScalarParameter() {defaultValue=0.25, parameterName=Glass Position Y}
  )
  n19 = NamedRerouteDeclaration(None=Add(A=Add(A=@n17, B=@n18)) {constB=0}) {name=Glass UV, nodeColor=<rgba 0 0 0 1>}
  n20 = MaterialFunctionCall!(UV Coord=@n19, LenghtScale, RadialScale) {materialFunction=/Game/Global/MaterialFunctions/TechArt/MF_PolarCoordinates}
  n21 = Power(Base=ComponentMask(None=@n20.Result) {r=true}, Exp=ScalarParameter() {defaultValue=0.15, parameterName=Glass UV Power}) {constExponent=0.2}
  n22 = ScalarParameter() {defaultValue=2, parameterName=Glass Repetition X}
  n23 = ComponentMask(None=@n20.Result) {g=true}
  n24 = ScalarParameter() {defaultValue=2, parameterName=Glass Repetition Y}
  n25 = TextureSample!(
   UVs=AppendVector(A=Multiply(A=@n21, B=@n22), B=Multiply(A=@n23, B=@n24) {constB=2})
   Tex
   "DDX(UVs)"=DDX(Value=@n19)
   "DDY(UVs)"=DDY(Value=@n19)
   Apply View MipBias
  ) {
   mipValueMode=TMVM_Derivative
   samplerSource=SSM_Wrap_WorldGroupSettings
   texture=/OfferCatalog/Art/SeasonPasses/Materials/S34/Texture/T_UI_CH6S2_Glass
  }
  n26 = NamedRerouteUsage()
  NamedRerouteDeclaration(None=Multiply(A=@n25.G, B=@n26)) {name=CracksFaces, nodeColor=<rgba 0.4531 1 0 1>}
  n27 = Power(Base=NamedRerouteUsage()) {constExponent=1}
  n28 = MaterialFunctionCall!(
   HeightTexture=Desaturation(None=OneMinus(None=@n27))
   TextureCoordinates=NamedRerouteUsage()
   ReferencePlane
   HeightRatioInput=ScalarParameter() {defaultValue=1, parameterName="HeightRatioInput (S)"}
   CameraVector=VectorParameter() {parameterName="CameraVector (V3)", defaultValue=<rgba 0.2 0 0 0>}.RGB
  ) {materialFunction=/Engine/Functions/Engine_MaterialFunctions02/Utility/BumpOffset_Advanced.BumpOffset_advanced}
  n29 = Constant()
  n30 = Constant() {r=1}
  NamedRerouteDeclaration(None=Multiply(A=MaterialFunctionCall(UV=@n28.UVs, GradientStart=@n29, GradientEnd=@n30) {materialFunction=&MF_UI_RadialGradient}.Result) {constB=0.9}) {name=Cracks Highlight, nodeColor=<rgba 0.0781 1 0 1>}
  NamedRerouteDeclaration(None=Multiply(A=@n25.B, B=@n26)) {name=CracksDirections, nodeColor=<rgba 0 0.25 1 1>}
  n31 = Add(A=ScreenPosition().ViewportUV, B=Constant2Vector() {r=-0.2}) {constB=-0.1}
  n32 = NamedRerouteUsage()
  n33 = TextureSample!(UVs=Add(A=@n31, B=Multiply(A=@n32) {constB=0.3}), Tex, Apply View MipBias) {samplerSource=SSM_Clamp_WorldGroupSettings, texture=/OfferCatalog/Art/SeasonPasses/Materials/S34/Texture/T_UI_CityLights}
  n34 = Saturate(None=Multiply(A=Power(Base=@n33.RGB) {constExponent=0.4}) {constB=0.9})
  n35 = Power(Base=NamedRerouteUsage()) {constExponent=0.5}
  n36 = Distance(
   A=MaterialFunctionCall(UVs=NamedRerouteUsage(), Texture Scale=Constant() {r=3}) {materialFunction=&ScaleUVsByCenter}.UVs
   B=Constant() {r=0.5}
  )
  n37 = Power(Base=OneMinus(None=Multiply(A=@n36) {constB=2})) {constExponent=0.5}
  NamedRerouteDeclaration(None=Multiply(A=LinearInterpolate(A=@n34, Alpha=@n35) {constAlpha=1, constB=0}, B=LinearInterpolate(Alpha=@n37) {constA=0.9})) {name=City Lights - Reflections, nodeColor=<rgba 0 0 0 1>}
  n38 = TextureSample(UVs=@n31) {texture=&T_UI_CityLights}
  n39 = Constant() {r=0.6}
  NamedRerouteDeclaration(None=Multiply(A=Desaturation(None=@n38.RGB, Fraction=@n39)) {constB=0.9}) {name=City Lights, nodeColor=<rgba 0 0 0 1>}
  n40 = NamedRerouteUsage()
  n41 = Min(A=Saturate(None=NamedRerouteUsage())) {constB=0.99}
  n42 = NamedRerouteUsage()
  n43 = LinearInterpolate(
   A=Multiply(A=@n42) {constB=0.4}
   B=Constant3Vector() {constant=<rgba 0.2197 0.379 0.4635 1>}
   Alpha=LinearInterpolate(Alpha=NamedRerouteUsage()) {constB=0.5}
  )
  n44 = NamedRerouteUsage()
  n45 = NamedRerouteUsage()
  n46 = Saturate(None=LinearInterpolate(A=LinearInterpolate(A=@n42, B=@n43, Alpha=@n44), B=@n45, Alpha=Multiply(A=@n45) {constB=0.2}))
  NamedRerouteDeclaration(None=MaterialFunctionCall!(
    Blend=Min(A=@n40) {constB=0.99}
    Base=MaterialFunctionCall(Blend=@n41, Base=@n46) {materialFunction=/Engine/Functions/Engine_MaterialFunctions03/Blends/Blend_ColorDodge}.Result
   ) {materialFunction=&Blend_ColorDodge}.Result) {name=Background, nodeColor=<rgba 0 0 0 1>}
  n47 = NamedRerouteUsage()
  n48 = LinearInterpolate(
   A=Constant3Vector() {constant=<rgba 0.0033 0.0016 0.0052 1>}
   B=Constant3Vector() {constant=<rgba 0.0153 0.0102 0.0208 1>}
   Alpha=Power(Base=@n47)
  )
  n49 = Multiply(A=@n47) {constB=0.4}
  n50 = Multiply(A=NamedRerouteUsage()) {constB=0.7}
  n51 = NamedRerouteUsage()
  n52 = Constant3Vector() {constant=<rgba 0 0.7519 1 1>}
  n53 = SmoothStep(Value=Max(A=NamedRerouteUsage(), B=Multiply(A=@n51, B=@n52))) {constMin=0.5}
  n54 = Constant3Vector() {constant=<rgba 0.0046 0.006 0.0156 1>}
  n55 = MaterialFunctionCall(UVs=NamedRerouteUsage(), Texture Scale=Constant() {r=2}) {materialFunction=&ScaleUVsByCenter}
  n56 = Multiply(A=Power(Base=MaterialFunctionCall(UV=@n55.UVs) {materialFunction=&MF_UI_RadialGradient}.Result)) {constB=0.4}
  n57 = MaterialFunctionCall(UV=ScreenPosition().ViewportUV) {materialFunction=&MF_UI_Random}
  n58 = Min(A=NamedRerouteUsage()) {constB=0.99}
  n59 = NamedRerouteUsage()
  n60 = MaterialFunctionCall(
   Blend=Min(A=NamedRerouteUsage())
   Base=MaterialFunctionCall(Blend=Min(A=@n59) {constB=0.999}, Base=NamedRerouteUsage()) {materialFunction=&Blend_ColorDodge}.Result
  ) {materialFunction=&Blend_ColorDodge}
  n61 = MaterialFunctionCall(
   Blend=Min(A=NamedRerouteUsage()) {constB=0.8}
   Base=MaterialFunctionCall(
    Blend=Multiply(A=@n57.Result) {constB=0.2}
    Base=MaterialFunctionCall(Blend=@n58, Base=@n60.Result) {materialFunction=&Blend_ColorDodge}.Result
   ) {materialFunction=&Blend_ColorDodge}.Result
  ) {materialFunction=&Blend_ColorDodge}
  n62 = MaterialFunctionCall(
   Blend=Power(Base=Saturate(None=@n50))
   Base=MaterialFunctionCall(Blend=Multiply(A=@n53) {constB=0.9}, Base=LinearInterpolate(A=@n61.Result, B=@n54, Alpha=@n56)) {materialFunction=&Blend_ColorDodge}.Result
  ) {materialFunction=&Blend_ColorDodge}
  n63 = NamedRerouteUsage()
  n64 = LinearInterpolate(
   A=Saturate(None=LinearInterpolate(A=@n62.Result, B=@n48, Alpha=@n49))
   B=VectorParameter() {group=Base Gradient, parameterName=CenterGradientColor, defaultValue=<rgba 0.0873 0.1669 0.599 1>}.RGB
   Alpha=Power(Base=SmoothStep(Value=@n63) {constMax=0.995, constMin=0.95})
  )
  n65 = Multiply(A=@n64) {constB=5}
  n66 = NamedRerouteUsage()
  n67 = Frac(None=Multiply(A=@n66) {constB=2})
  n68 = Multiply(A=NamedRerouteUsage()) {constB=0.5}
  n69 = Multiply(
   A=TextureSample(UVs=Add(A=@n67, B=@n68)) {samplerType=SAMPLERTYPE_LinearColor, texture=/OfferCatalog/Art/SeasonPasses/Materials/S34/Texture/T_UI_CH6S2_CloudNoise}.B
   B=TextureSample(UVs=Add(A=@n66) {constB=0.2}) {samplerType=SAMPLERTYPE_LinearColor, texture=&T_UI_CH6S2_CloudNoise}.R
  )
  n70 = SmoothStep(Value=OneMinus(None=NamedRerouteUsage())) {constMin=0.8}
  NamedRerouteDeclaration(None=LinearInterpolate(A=@n64, B=Saturate(None=@n65), Alpha=Multiply(A=@n69, B=@n70) {constB=0.5})) {name=Middleground, nodeColor=<rgba 0 0 0 1>}
  n71 = Add(A=NamedRerouteUsage(), B=Multiply(A=NamedRerouteUsage()) {constB=0.2})
  n72 = Constant2Vector() {g=0.35, r=0.5}
  n73 = Constant2Vector() {g=2.2, r=1.6}
  n74 = Constant() {r=2}
  NamedRerouteDeclaration(None=OneMinus(None=MaterialFunctionCall(UVs=@n71, CenterPosition=@n72, Radius=@n73, Density=@n74) {materialFunction=&RadialGradientExponential}.RadialGradientExponential)) {name=Bottom Gradient, nodeColor=<rgba 0 0 0 1>}
  n75 = Power(Base=NamedRerouteUsage()) {constExponent=0.5}
  n76 = MaterialFunctionCall!(UV=ScreenPosition().ViewportUV, Twist Amount=Constant() {r=1}) {materialFunction=/FortUILibrary/MaterialFunctions/Effects/MF_UI_Twist}
  n77 = Constant() {r=0.4}
  n78 = TextureSample(UVs=Panner(Coordinate=MaterialFunctionCall(UVs=@n76.Twist UV, Texture Scale=@n77) {materialFunction=&ScaleUVsByCenter}.UVs) {speedX=-0.25}) {samplerSource=SSM_Wrap_WorldGroupSettings, texture=&T_UI_CityLights}
  NamedRerouteDeclaration(None=Saturate(None=Multiply(A=@n75, B=@n78.RGB))) {name=Cracks Valleys Highlights, nodeColor=<rgba 0 0 0 1>}
  n79 = Power(Base=MaterialFunctionCall(UVs=ScreenPosition().ViewportUV, GradientStart=Constant() {r=-1}, RotationAngle=Constant() {r=0.78}, GradientOffset=Constant() {r=0.1}) {materialFunction=&MF_UI_LinearGradient}.Result)
  n80 = Multiply(A=Constant3Vector() {constant=<rgba 0.0008 0.0021 0.0052 1>}, B=@n79)
  n81 = ScreenPosition()
  n82 = Constant2Vector() {g=0.5, r=1.1}
  n83 = Add(A=@n81.ViewportUV, B=@n82) {constB=0}
  n84 = Constant() {r=1.2}
  n85 = OneMinus(None=Power(Base=MaterialFunctionCall!(UV=@n83, Size=@n84, OutlineThickness, OutlineAlignment, GlowMin, GlowMax) {materialFunction=&MF_UI_Circle}.SDF) {constExponent=4})
  n86 = MaterialFunctionCall(UV=Add(A=@n81.ViewportUV, B=@n82) {constB=0}, Size=Constant() {r=0.1}) {materialFunction=&MF_UI_Circle}
  n87 = LinearInterpolate(
   A=NamedRerouteUsage()
   B=Multiply(A=Constant3Vector() {constant=<rgba 0.0053 0.0036 0.008 1>}, B=Power(Base=@n86.SDF))
   Alpha=Multiply(A=Saturate(None=@n85))
  )
  NamedRerouteDeclaration(None=MaterialFunctionCall(Base=LinearInterpolate(A=@n87, B=@n80, Alpha=@n79), Blend=Constant() {r=0.8}) {materialFunction=&Blend_Overlay}.Result) {name=Foreground, nodeColor=<rgba 0 0 0 1>}
  n88 = ComponentMask(None=ScreenPosition().ViewportUV) {r=true}
  n89 = Panner(Coordinate=Add(A=ScreenPosition().ViewportUV, B=Constant2Vector() {g=-0.1}) {constB=0.2}, Time=Multiply(A=Time(), B=Constant() {r=0.2})) {speedX=1}
  n90 = AppendVector(A=ScalarParameter() {defaultValue=0.35, parameterName=Police Frequency}, B=Constant() {r=1})
  n91 = MaterialFunctionCall(UVs=Frac(None=Multiply(A=@n89, B=@n90) {constB=0.5}), Texture Scale=Constant2Vector() {g=2, r=0.4}) {materialFunction=&ScaleUVsByCenter}
  n92 = ScalarParameter() {defaultValue=-0.82, parameterName=Police Lights Size}
  n93 = OneMinus(None=MaterialFunctionCall(UV=@n91.UVs, Size=@n92) {materialFunction=&MF_UI_Circle}.SDF)
  n94 = MaterialFunctionCall(Value=Time()) {materialFunction=&MF_UI_PingPong}
  n95 = Multiply(A=Multiply(A=Saturate(None=@n93), B=Constant3Vector() {constant=<rgba 1 0.0005 0.1438 1>}), B=@n94.Result)
  n96 = Constant2Vector() {r=0.1}
  n97 = OneMinus(None=MaterialFunctionCall(UV=Add(A=@n91.UVs, B=@n96) {constB=0.2}, Size=@n92) {materialFunction=&MF_UI_Circle}.SDF)
  n98 = Multiply(A=@n94.Inverse Result, B=Multiply(A=Saturate(None=@n97), B=Constant3Vector() {constant=<rgba 0 0.4411 1 1>}))
  n99 = Multiply(A=SmoothStep(Value=Max(A=@n95, B=@n98)) {constMax=0.2}) {constB=2}
  NamedRerouteDeclaration(None=Multiply(A=SmoothStep(Value=@n88) {constMax=0.4, constMin=0.2}, B=Saturate(None=@n99))) {name=Police Lights Base, nodeColor=<rgba 0 0 0 1>}
  n100 = ComponentMask(None=ScreenPosition().ViewportUV) {r=true}
  n101 = NamedRerouteUsage()
  n102 = Panner(Coordinate=Add(A=ScreenPosition().ViewportUV, B=Multiply(A=@n101) {constB=0.2}), Time=Multiply(A=Time(), B=Constant() {r=0.2})) {speedX=1}
  n103 = AppendVector(A=ScalarParameter() {defaultValue=0.35, parameterName=Police Frequency}, B=Constant() {r=1})
  n104 = MaterialFunctionCall(UVs=Frac(None=Multiply(A=@n102, B=@n103) {constB=0.5}), Texture Scale=Constant2Vector() {g=2, r=0.4}) {materialFunction=&ScaleUVsByCenter}
  n105 = ScalarParameter() {defaultValue=-0.82, parameterName=Police Lights Size}
  n106 = OneMinus(None=MaterialFunctionCall(UV=@n104.UVs, Size=@n105) {materialFunction=&MF_UI_Circle}.SDF)
  n107 = MaterialFunctionCall(Value=Time()) {materialFunction=&MF_UI_PingPong}
  n108 = Multiply(A=Multiply(A=Saturate(None=@n106), B=Constant3Vector() {constant=<rgba 1 0.0005 0.1438 1>}), B=@n107.Result)
  n109 = Constant2Vector() {r=0.1}
  n110 = OneMinus(None=MaterialFunctionCall(UV=Add(A=@n104.UVs, B=@n109) {constB=0.2}, Size=@n105) {materialFunction=&MF_UI_Circle}.SDF)
  n111 = Multiply(A=@n107.Inverse Result, B=Multiply(A=Saturate(None=@n110), B=Constant3Vector() {constant=<rgba 0 0.4411 1 1>}))
  n112 = Multiply(A=SmoothStep(Value=Max(A=@n108, B=@n111)) {constMax=0.2}) {constB=2}
  NamedRerouteDeclaration(None=Multiply(A=SmoothStep(Value=@n100) {constMax=0.4, constMin=0.2}, B=Saturate(None=@n112))) {name=Police Lights Reflection, nodeColor=<rgba 0 0 0 1>}
  n113 = NamedRerouteUsage()
  n114 = Multiply(A=@n113) {constB=2}
  n115 = Multiply(A=TextureSample(UVs=Frac(None=@n114)) {samplerType=SAMPLERTYPE_LinearColor, texture=&T_UI_CH6S2_CloudNoise}.G) {constB=6}
  n116 = Multiply(A=NamedRerouteUsage()) {constB=0.8}
  n117 = Multiply(A=TextureSample(UVs=Add(A=@n113, B=@n116)) {samplerType=SAMPLERTYPE_LinearColor, texture=&T_UI_CH6S2_CloudNoise}.R) {constB=1.4}
  n118 = NamedRerouteUsage()
  n119 = Multiply(A=Saturate(None=Multiply(A=@n115, B=@n117)), B=OneMinus(None=SmoothStep(Value=@n118) {constMin=0.9}))
  n120 = OneMinus(None=NamedRerouteUsage())
  NamedRerouteDeclaration(None=Multiply(A=Saturate(None=@n119), B=Saturate(None=@n120))) {name=Grunge Texture, nodeColor=<rgba 0 0.7891 1 1>}
  n121 = ScreenPosition()
  n122 = Constant() {r=0.2}
  n123 = Multiply(
   A=NamedRerouteUsage()
   B=LinearInterpolate(Alpha=MaterialFunctionCall(Speed=@n122) {materialFunction=&DebugTimeSine}.Result) {constA=0.2, constB=0.35}
  ) {constB=0.3}
  n124 = MaterialFunctionCall(UVs=Add(A=Add(A=@n121.ViewportUV, B=@n123), B=Constant2Vector() {g=-0.5, r=0.5}), Texture Scale=Constant2Vector() {g=6, r=8}) {materialFunction=&ScaleUVsByCenter}
  n125 = Time()
  n126 = Constant() {r=0.1}
  n127 = Power(
   Base=Power(Base=MaterialFunctionCall(UV=@n124.UVs) {materialFunction=&MF_UI_RadialGradient}.Result) {constExponent=3}
   Exp=LinearInterpolate(Alpha=MaterialFunctionCall(Value=@n125, Scale=@n126) {materialFunction=&MF_UI_PingPong}.Result) {constA=0.5}
  )
  n128 = Multiply(A=MaterialFunctionCall() {materialFunction=&MF_UI_Random}.Result, B=@n127)
  NamedRerouteDeclaration(None=Add(A=@n127, B=Multiply(A=@n128) {constB=0.1})) {name=Left Light, nodeColor=<rgba 0 0 0 1>}
  n129 = NamedRerouteUsage()
  n130 = Max(A=Multiply(A=@n129) {constB=0.2}, B=SmoothStep(Value=OneMinus(None=@n129)) {constMin=0.9})
  n131 = MaterialFunctionCall(UVs=NamedRerouteUsage(), Radius=Constant() {r=1.5}) {materialFunction=&RadialGradientExponential}
  n132 = OneMinus(None=@n131.RadialGradientExponential)
  n133 = Saturate(None=Power(Base=Saturate(None=@n130), Exp=LinearInterpolate(Alpha=@n132) {constA=1, constB=80}) {constExponent=3})
  n134 = NamedRerouteDeclaration(None=@n131.RadialGradientExponential) {name=Glass Mask, nodeColor=<rgba 1 0 0.6562 1>}
  n135 = TextureSample(UVs=ScreenPosition().ViewportUV) {samplerType=SAMPLERTYPE_LinearColor, texture=&T_UI_CH6S2_CloudNoise}
  NamedRerouteDeclaration(None=Multiply(A=Multiply(A=@n133, B=@n134), B=LinearInterpolate(Alpha=@n135.R) {constA=0.2})) {name=Glass Cracks, nodeColor=<rgba 0 0 0 1>}
  NamedRerouteDeclaration(None=Power(Base=Saturate(None=@n25.R)) {constExponent=1}) {name=Glass Base, nodeColor=<rgba 0 0 0 1>}
  n136 = VectorParameter() {parameterName=ShineExtents, defaultValue=<rgba -1 1 0 0>}
  n137 = Multiply(A=Time()) {constB=0.1}
  n138 = Desaturation(None=OneMinus(None=NamedRerouteUsage()))
  n139 = MaterialFunctionCall(UVs=NamedRerouteUsage(), "Rotation Angle (0-1)"=Constant() {r=0.1}) {materialFunction=&CustomRotator}
  n140 = ScalarParameter() {defaultValue=1, parameterName="HeightRatioInput (S)"}
  n141 = VectorParameter() {parameterName="CameraVector (V3)", defaultValue=<rgba 0.2 0 0 0>}
  n142 = MaterialFunctionCall!(
   ShineOffset=LinearInterpolate(A=@n136.R, B=@n136.R, Alpha=Frac(None=@n137)) {constA=1, constB=-1}
   Rotation=Constant() {r=1}
   RotationCenter=VectorParameter() {parameterName="RotationCenter (V2)", defaultValue=<rgba 0.5 0.5 0 0>}.RGB
   UVs=NamedRerouteDeclaration(None=MaterialFunctionCall(HeightTexture=@n138, TextureCoordinates=@n139.Rotated Values, HeightRatioInput=@n140, CameraVector=@n141.RGB) {materialFunction=&BumpOffset_advanced}.UVs) {name=Glass Bump, nodeColor=<rgba 0 0 0 1>}
  ) {materialFunction=/Game/Athena/HUD/Materials/Victory/MF_Shine}
  n143 = MaterialFunctionCall(UVs=@n139.Rotated Values, GradientStart=Constant() {r=-0.5}, GradientEnd=Constant() {r=0.5}) {materialFunction=&MF_UI_LinearGradient}
  NamedRerouteDeclaration(None=Multiply(A=Multiply(A=@n142.Result, B=@n143.Result), B=Constant3Vector() {constant=<rgba 0.3562 0.4991 0.526 1>})) {name=Cracks Shine, nodeColor=<rgba 0 0 0 1>}
  n144 = LinearInterpolate(A=Constant3Vector() {constant=<rgba 0.9449 0 1 1>}, B=Constant3Vector() {constant=<rgba 0.0527 0.0118 0.1979 1>}, Alpha=NamedRerouteUsage())
  n145 = Multiply(A=@n144, B=NamedRerouteUsage())
  n146 = MaterialFunctionCall(UV=ScreenPosition().ViewportUV) {materialFunction=&MF_UI_Random}
  n147 = Multiply(A=Multiply(A=Multiply(A=@n146.Result, B=@n145), B=@n144)) {constB=0.15}
  NamedRerouteDeclaration(None=Saturate(None=Add(A=@n145, B=@n147))) {name=Left Light Color, nodeColor=<rgba 1 0.0646 0.8727 1>}
 ]}
