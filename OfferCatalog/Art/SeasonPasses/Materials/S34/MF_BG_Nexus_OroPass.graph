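# UI material function that builds the animated, themeable Nexus OroPass background colour for the Season 34 season pass screens.
# Five colour parameters and two grunge textures feed layered lerps, finished with an adjustable radial vignette and a saturated additive layer.
# Named reroutes carry shared UVs and masks, including an MSDF sun with adjustable placement.
# God rays come from stretched, panning noise under radial falloff; smoke combines flowmapped and scrolling noise, faded by a linear gradient.
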MaterialFunction /OfferCatalog/Art/SeasonPasses/Materials/S34/MF_BG_Nexus_OroPass
graph {nodes [
  n1 = Multiply(A=Power(Base=NamedRerouteUsage()) {constExponent=1.5}, B=NamedRerouteUsage())
  n2 = Constant() {r=0.65}
  n3 = NamedRerouteUsage()
  n4 = Constant() {r=0.3072}
  n5 = Constant() {r=0.564}
  n6 = LinearInterpolate(
   A=NamedRerouteUsage()
   B=NamedRerouteUsage()
   Alpha=Multiply(
    A=MaterialFunctionCall(Input=@n3, Input Low=@n4, Input High=@n5) {materialFunction=&RemapValue_0-1}.Result
    B=Constant() {r=1.0258}
   ) {constB=0.3}
  )
  n7 = LinearInterpolate(A=Constant3Vector() {constant=<rgba 0.0406 0.0038 0.005 1>}, B=NamedRerouteUsage(), Alpha=NamedRerouteUsage())
  n8 = NamedRerouteUsage()
  n9 = LinearInterpolate(
   A=Saturate(None=LinearInterpolate(A=@n6, B=@n1, Alpha=@n2))
   B=LinearInterpolate(A=@n7, B=Multiply(A=@n7, B=@n8), Alpha=Constant() {r=0.7792}) {constAlpha=0.9}
   Alpha=Multiply(A=NamedRerouteUsage()) {constB=0.6}
  )
  n10 = LinearInterpolate(A=@n9, B=NamedRerouteUsage(), Alpha=Constant() {r=0.026}) {constA=0.1, constAlpha=0.1, constB=0.74}
  n11 = MaterialFunctionCall(
   Input=MaterialFunctionCall(UV Channel=NamedRerouteUsage()) {materialFunction=&LinearGradient}.VGradient
   Input Low=Constant() {r=0.5}
   Input High=Constant() {r=1}
  ) {materialFunction=&RemapValue_0-1}
  n12 = MaterialFunctionCall(UVs=NamedRerouteUsage(), Texture Scale=ScalarParameter() {group=Vignette, parameterName=VignetteScale}) {materialFunction=&ScaleUVsByCenter}
  n13 = AppendVector(A=ScalarParameter() {group=Vignette, parameterName=VignetteX}, B=ScalarParameter() {group=Vignette, parameterName=VignetteY})
  n14 = SmoothStep(Value=MaterialFunctionCall(UV=Add(A=@n12.UVs, B=@n13), GradientStart=Constant(), GradientEnd=Constant() {r=1}) {materialFunction=&MF_UI_RadialGradient}.Result)
  n15 = LinearInterpolate(
   A=LinearInterpolate(A=LinearInterpolate(A=@n9, B=@n10, Alpha=@n11.Result), B=NamedRerouteUsage(), Alpha=NamedRerouteUsage())
   Alpha=Multiply(A=OneMinus(None=@n14)) {constB=0.9}
  ) {constB=0}
  n16 = NamedRerouteUsage()
  FunctionOutput(None=Saturate(None=Add(A=@n15, B=@n16))) {outputName=Colour}
  MaterialFunctionCall!("Use Local UV (Debug)", Texture Size, Scale Pivot Point) {materialFunction=/Game/Environments/FrontEnd/Blueprints/ItemPreview/VaultWorldTransitions/MaterialFunctions/MF_UI_VaultWorld_UV}
  n17 = MaterialFunctionCall() {materialFunction=&MF_UI_UVScaling}
  NamedRerouteDeclaration(None=@n17.Scaled to Fit) {name=FocalUV, nodeColor=<rgba 0.2656 0 1 1>}
  NamedRerouteDeclaration(None=VectorParameter() {parameterName=Colour1, defaultValue=<rgba 0.4852 0.0357 0.0268 1>}.RGB) {name=Colour Theme 1, nodeColor=<rgba 0.0547 1 0 1>}
  NamedRerouteDeclaration(None=TextureSample(UVs=NamedRerouteUsage()) {texture=/OfferCatalog/Art/SeasonPasses/Materials/S34/Texture/T_UI_JunoPass_S2_BackgroundGrunge_Layer1}.R) {name=BGTexture01, nodeColor=<rgba 1 0.375 0 1>}
  n18 = TextureSample(UVs=NamedRerouteUsage()) {samplerType=SAMPLERTYPE_Alpha, texture=/OfferCatalog/Art/SeasonPasses/Materials/S34/Texture/T_UI_JunoPass_S2_BackgroundGrunge_Layer2}
  NamedRerouteDeclaration(None=@n18.R) {name=BGTexture02, nodeColor=<rgba 0.0547 0 1 1>}
  n19 = ScalarParameter() {defaultValue=1.92, group=Sun, parameterName=SunSize}
  n20 = ScalarParameter() {group=Sun, parameterName=X Position}
  n21 = ScalarParameter() {group=Sun, parameterName=Y Position}
  n22 = TextureSample(
   UVs=Add(
    A=MaterialFunctionCall(UVs=@n17.Scaled to Fit, Texture Scale=@n19) {materialFunction=&ScaleUVsByCenter}.UVs
    B=AppendVector(A=@n20, B=@n21)
   )
   Tex=TextureObject() {samplerType=SAMPLERTYPE_LinearColor, texture=/OfferCatalog/Art/SeasonPasses/Materials/S34/Texture/T_UI_Sun_Test}
  )
  NamedRerouteDeclaration(None=MaterialFunctionCall(
    DistanceField=MaterialFunctionCall!(MSDF RGB=@n22.RGB) {materialFunction=/FortUILibrary/MaterialFunctions/Shapes/Modifiers/MF_UI_MSDFtoSDF}.SDF
    Threshold=Constant() {r=0.5}
    Smoothing=Constant()
   ) {materialFunction=&MF_UI_AntiAliasing}.Outer) {name=Sun Mask, nodeColor=<rgba 1 0 0.6094 1>}
  n23 = MaterialFunctionCall() {materialFunction=&MF_UI_UVScaling}
  n24 = ScalarParameter() {group=Focal, parameterName=Focal Scale}
  n25 = ScalarParameter() {group=Focal, parameterName=Focal X Position}
  n26 = ScalarParameter() {group=Focal, parameterName=Focal Y Position}
  NamedRerouteDeclaration(None=Add(
    A=MaterialFunctionCall(UVs=@n23.Scaled To Fill, Texture Scale=@n24) {materialFunction=&ScaleUVsByCenter}.UVs
    B=AppendVector(A=@n25, B=@n26)
   )) {name=Focal Mask, nodeColor=<rgba 0.8984 0 1 1>}
  NamedRerouteDeclaration(None=VectorParameter() {parameterName=Colour2, defaultValue=<rgba 0.7454 0.3663 0.0232 1>}.RGB) {name=Colour Theme 2, nodeColor=<rgba 1 0 0.375 1>}
  n27 = NamedRerouteUsage()
  n28 = Add(
   A=MaterialFunctionCall(UVs=@n27, "Rotation Angle (0-1)"=Constant() {r=0.5984}) {materialFunction=&CustomRotator}.Rotated Values
   B=Constant2Vector() {g=-0.45}
  )
  n29 = Add(A=@n28, B=Time())
  n30 = Constant2Vector() {g=40, r=1}
  n31 = Constant() {r=-0.95}
  n32 = TextureSample(UVs=Panner(
    Coordinate=MaterialFunctionCall(UVs=@n29, Texture Scale=@n30) {materialFunction=&ScaleUVsByCenter}.UVs
    Speed=MaterialFunctionCall(X=@n31) {materialFunction=&MakeFloat2}.Result
   )) {samplerType=SAMPLERTYPE_LinearGrayscale, texture=/Game/Global/Textures/Default/Noise/LowResBlurredNoise_NoMips}
  n33 = Constant() {r=1.3}
  n34 = MaterialFunctionCall(UVs=@n28, Texture Scale=Constant2Vector() {g=3, r=4}) {materialFunction=&ScaleUVsByCenter}
  n35 = Multiply(A=Multiply(A=@n34.UVs, B=OneMinus(None=@n34.UVs))) {constB=4}
  n36 = Multiply(A=Multiply(A=ComponentMask(None=@n35) {r=true}, B=ComponentMask(None=@n35) {g=true})) {constB=4}
  n37 = ComponentMask(None=@n34.UVs) {g=true}
  n38 = LinearInterpolate(A=@n37, B=Multiply(A=@n37, B=OneMinus(None=@n37)), Alpha=Constant())
  n39 = Constant() {r=1.5}
  n40 = Multiply(A=Multiply(A=Power(Base=@n32.R, Exp=@n33), B=Constant() {r=3.7}), B=Multiply(A=Saturate(None=@n36), B=Power(Base=@n38, Exp=@n39)))
  n41 = Constant2Vector() {g=0.5, r=-0.5}
  n42 = MaterialFunctionCall(UV=MaterialFunctionCall(UVs=Add(A=@n27, B=@n41), Texture Scale=Constant() {r=5.5}) {materialFunction=&ScaleUVsByCenter}.UVs) {materialFunction=&MF_UI_RadialGradient}
  n43 = Saturate(None=Power(Base=Multiply(A=@n40, B=@n42.Result)) {constExponent=2.25})
  n44 = Multiply(A=LinearInterpolate(A=NamedRerouteUsage(), B=@n43) {constAlpha=0.4}, B=@n43)
  n45 = LinearInterpolate(A=NamedRerouteUsage(), B=Constant3Vector() {constant=<rgba 1 0.9936 0.6365 1>}, Alpha=@n44)
  n46 = Multiply(A=@n44, B=Constant() {r=0.8})
  NamedRerouteDeclaration(None=Saturate(None=Multiply(A=@n45, B=@n46))) {name=GodRays, nodeColor=<rgba 1 0 0.1875 1>}
  NamedRerouteDeclaration(None=VectorParameter() {parameterName=Colour 3, defaultValue=<rgba 0.1119 0.0561 0.0037 1>}.RGB) {name=Colour Theme 3, nodeColor=<rgba 0 0.6953 1 1>}
  n47 = LinearInterpolate(A=NamedRerouteUsage(), B=NamedRerouteUsage(), Alpha=LinearInterpolate(Alpha=NamedRerouteUsage()) {constA=0.4, constB=0.8}) {constAlpha=0.25}
  NamedRerouteDeclaration(None=Saturate(None=Multiply(A=@n47) {constB=1.6})) {name=RockTextureMask, nodeColor=<rgba 1 0.9375 0 1>}
  NamedRerouteDeclaration(None=VectorParameter() {parameterName=Colour4, defaultValue=<rgba 0.01 0.0067 0 1>}.RGB) {name=Colour Theme 4, nodeColor=<rgba 1 0.9844 0 1>}
  NamedRerouteDeclaration(None=@n23.ScaleY) {name=Base UV, nodeColor=<rgba 0.7344 0 1 1>}
  NamedRerouteDeclaration(None=@n23.ScaleY) {name=Scale Y UV, nodeColor=<rgba 0 1 0.4141 1>}
  NamedRerouteDeclaration(None=VectorParameter() {parameterName=Colour 5, defaultValue=<rgba 0.001 0.0006 0.0001 1>}.RGB) {name=Colour Theme 5, nodeColor=<rgba 0 1 0.7656 1>}
  n48 = Constant() {r=1}
  n49 = MaterialFunctionCall() {materialFunction=&MF_UI_UVScaling}
  n50 = Constant() {r=0.01}
  n51 = Constant() {r=0.03}
  n52 = MaterialFunctionCall!(
   Texture 2D=TextureObject() {samplerType=SAMPLERTYPE_LinearGrayscale, texture=&LowResBlurredNoise_NoMips}
   Texture UVs=MaterialFunctionCall(UVs=NamedRerouteUsage(), Texture Scale=Constant() {r=0.5}) {materialFunction=&ScaleUVsByCenter}.UVs
   FlowMap Texture 2D=TextureObject() {samplerType=SAMPLERTYPE_LinearColor, texture=/Game/Global/Textures/Default/Flowmaps/T_SwirlFlow_01}
   FlowMap UVs=Panner(Coordinate=Multiply(A=@n49.Scaled To Fill, B=@n48), Speed=AppendVector(A=@n50, B=@n51))
   FlowIntensity=Constant() {r=0.2}
   FlowSpeed
  ) {materialFunction=/FortUILibrary/MaterialFunctions/Effects/MF_UI_FlowMap}
  n53 = Multiply(A=Time(), B=Constant() {r=2})
  n54 = TextureSample(UVs=Panner(Coordinate=@n49.Scaled To Fill, Time=@n53) {speedX=0.015, speedY=0.013}) {samplerSource=SSM_Wrap_WorldGroupSettings, texture=/OfferCatalog/Art/SeasonPasses/Materials/S34/Texture/T_UI_JunoPass_Noise}
  n55 = TextureSample(UVs=Panner(Coordinate=@n49.Scaled To Fill, Time=@n53) {speedX=0.05, speedY=0.015}) {samplerType=SAMPLERTYPE_Grayscale, texture=&LowResBlurredNoise}
  n56 = Saturate(None=Multiply(A=Add(A=@n54.R, B=@n55.R)) {constB=1.4})
  n57 = NamedRerouteDeclaration(None=LinearInterpolate(B=LinearInterpolate(A=@n52.Result, Alpha=@n56), Alpha=@n52.Result)) {name=SmokeMask, nodeColor=<rgba 1 0.1406 0 1>}
  n58 = Constant()
  n59 = MaterialFunctionCall(
   UVs=NamedRerouteUsage()
   GradientStart=ScalarParameter() {defaultValue=0.7356, desc=Smoke, group=Smoke, parameterName=SmokeGradientStart, sliderMax=1}
   GradientEnd=ScalarParameter() {defaultValue=2.5, group=Smoke, parameterName=SmokeGradientEnd}
   RotationAngle=MaterialFunctionCall(Input=Constant() {r=90}, Input Low=@n58, Input High=Constant() {r=-360}, Target Low=@n58, Target High=Constant() {r=1}) {materialFunction=&RemapValueRange}.Result
  ) {materialFunction=&MF_UI_LinearGradient}
  NamedRerouteDeclaration(None=Multiply(A=Multiply(A=@n57, B=@n59.Result), B=Constant() {r=0.8468})) {name=FadedSmokeMask, nodeColor=<rgba 1 0.1406 0 1>}
 ]}
